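# First-person firing sound cue for the DMR22 weapon.
# A mixer layers randomized gunshot, mechanism and quieter, delayed shell-casing samples, each with its own modulation; the gunshot passes through two modulators.
# The root SoundClass node overrides to the third-person fire class, although the cue itself uses the first-person class.
# It references first-person attenuation and gunfire concurrency assets, with overrides specifying no spatialization and per-owner, stop-oldest limiting.
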SoundCue /CobraDMR/DMR/Sounds/DMR22/DMR22_Fire_1P_Cue
root=SoundClass_3
nodes [
 {id=Delay_0, children [Modulator_0], properties {delayMax=0.4, delayMin=0.4}}
 {id=Mixer_0, children [Modulator_3,Modulator_1,Delay_0], properties {inputVolume [1,1.1,1.7]}}
 {id=Modulator_0, children [Random_1], properties {pitchMax=1, pitchMin=1, volumeMax=0.05, volumeMin=0.05}}
 {id=Modulator_1, children [Random_2], properties {pitchMax=0.95, volumeMax=1, volumeMin=1}}
 {id=Modulator_2, children [Random_0]}
 {id=Modulator_3, children [Modulator_2], properties {pitchMax=1, pitchMin=0.98, volumeMax=1.2, volumeMin=1.2}}
 {
  id=Random_0
  children [WavePlayer_2, WavePlayer_6, WavePlayer_12, WavePlayer_13, WavePlayer_14, WavePlayer_15]
  properties {weights [1,1,1,1,1,1]}
 }
 {
  id=Random_1
  children [WavePlayer_37, WavePlayer_38, WavePlayer_39, WavePlayer_40, WavePlayer_41, WavePlayer_42, WavePlayer_43]
  properties {weights [1,1,1,1,1,1,1]}
 }
 {id=Random_2, children [WavePlayer_7, WavePlayer_8, WavePlayer_9, WavePlayer_10, WavePlayer_11], properties {weights [1,1,1,1,1]}}
 {id=SoundClass_3, children [Mixer_0], properties {soundClassOverride=&Fort_WeaponRanged_Fire_3P_PSM}}
 {id=WavePlayer_10, properties {soundWaveAssetPtr=/CobraDMR/DMR/Sounds/DMR_Mech_04}}
 {id=WavePlayer_11, properties {soundWaveAssetPtr=/CobraDMR/DMR/Sounds/DMR_Mech_05}}
 {id=WavePlayer_12, properties {soundWaveAssetPtr=&/DMR22_Fire_1P_02}}
 {id=WavePlayer_13, properties {soundWaveAssetPtr=&/DMR22_Fire_1P_03}}
 {id=WavePlayer_14, properties {soundWaveAssetPtr=&/DMR22_Fire_1P_04}}
 {id=WavePlayer_15, properties {soundWaveAssetPtr=&/DMR22_Fire_1P_05}}
 {id=WavePlayer_2, properties {soundWaveAssetPtr=&/DMR22_Fire_1P_06}}
 {id=WavePlayer_37, properties {soundWaveAssetPtr=/CobraDMR/DMR/Sounds/DMR_ShellCasings_01}}
 {id=WavePlayer_38, properties {soundWaveAssetPtr=/CobraDMR/DMR/Sounds/DMR_ShellCasings_02}}
 {id=WavePlayer_39, properties {soundWaveAssetPtr=/CobraDMR/DMR/Sounds/DMR_ShellCasings_03}}
 {id=WavePlayer_40, properties {soundWaveAssetPtr=/CobraDMR/DMR/Sounds/DMR_ShellCasings_04}}
 {id=WavePlayer_41, properties {soundWaveAssetPtr=/CobraDMR/DMR/Sounds/DMR_ShellCasings_05}}
 {id=WavePlayer_42, properties {soundWaveAssetPtr=/CobraDMR/DMR/Sounds/DMR_ShellCasings_06}}
 {id=WavePlayer_43, properties {soundWaveAssetPtr=/CobraDMR/DMR/Sounds/DMR_ShellCasings_07}}
 {id=WavePlayer_6, properties {soundWaveAssetPtr=&/DMR22_Fire_1P_01}}
 {id=WavePlayer_7, properties {soundWaveAssetPtr=/CobraDMR/DMR/Sounds/DMR_Mech_01}}
 {id=WavePlayer_8, properties {soundWaveAssetPtr=/CobraDMR/DMR/Sounds/DMR_Mech_02}}
 {id=WavePlayer_9, properties {soundWaveAssetPtr=/CobraDMR/DMR/Sounds/DMR_Mech_03}}
]
properties {
 attenuationSettings=&Fort_Weapon_Projectile_FirstPerson_Athena_Att
 bDebug=true
 maxDistance=2097152
 soundClassObject=&Fort_WeaponRanged_Fire_1P_PSM
 volumeMultiplier=0.8
 attenuationOverrides {bSpatialize=false}
 concurrencyOverrides {bLimitToOwner=true, maxCount=3, resolutionRule=StopOldest}
 concurrencySet [&Weapon_Gunfire_FirstPerson_Conc]
}
