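# Third-person gunfire sound cue for the DMR22 weapon, layering close, medium and distant shots.
# Each layer randomly picks a wave, sets volume through a modulator and uses its own attenuation, with pitch variation on close and medium.
# Sound class overrides separate close from medium and distant before mixing.
# A quality-level node cuts the distant pool from six waves to three on two of four levels.
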SoundCue /CobraDMR/DMR/Sounds/DMR22/DMR22_Fire_3P_Cue
root=Mixer_2
nodes [
 {id=Attenuation_3, children [Modulator_0], properties {attenuationSettings=&Fort_Weapon_Projectile_Close_Athena_Att}}
 {
  id=Attenuation_4
  children [Modulator_2]
  properties {attenuationSettings=&Fort_Weapon_Projectile_Distant_Mono_Athena_Att}
 }
 {
  id=Attenuation_5
  children [Modulator_4]
  properties {attenuationSettings=&Fort_Weapon_Projectile_Medium_Mono_Athena_Att}
 }
 {id=Mixer_2, children [Modulator_3,SoundClass_4], properties {inputVolume [1,1.5]}}
 {id=Mixer_3, children [SoundClass_3,SoundClass_5], properties {inputVolume [1,1.8]}}
 {id=Modulator_0, children [Random_2], properties {pitchMax=1.02, pitchMin=0.98, volumeMax=0.8, volumeMin=0.8}}
 {id=Modulator_2, children [QualityLevel_0], properties {volumeMax=0.25, volumeMin=0.25}}
 {id=Modulator_3, children [Mixer_3]}
 {id=Modulator_4, children [Random_1], properties {pitchMax=0.9, pitchMin=0.85, volumeMax=0.2, volumeMin=0.2}}
 {id=QualityLevel_0, children [Random_0,Random_0,Random_3,Random_3]}
 {
  id=Random_0
  children [WavePlayer_36, WavePlayer_37, WavePlayer_38, WavePlayer_39, WavePlayer_40, WavePlayer_41]
  properties {weights [1,1,1,1,1,1]}
 }
 {
  id=Random_1
  children [WavePlayer_30, WavePlayer_31, WavePlayer_32, WavePlayer_33, WavePlayer_34, WavePlayer_35]
  properties {weights [1,1,1,1,1,1]}
 }
 {
  id=Random_2
  children [WavePlayer_14, WavePlayer_25, WavePlayer_26, WavePlayer_27, WavePlayer_28, WavePlayer_29]
  properties {weights [1,1,1,1,1,1]}
 }
 {id=Random_3, children [WavePlayer_39, WavePlayer_40, WavePlayer_41], properties {weights [1,1,1]}}
 {
  id=SoundClass_3
  children [Attenuation_3]
  properties {soundClassOverride=&Fort_WeaponRanged_Fire_3P_PSM}
 }
 {
  id=SoundClass_4
  children [Attenuation_4]
  properties {soundClassOverride=&Fort_WeaponRanged_Fire_Distant}
 }
 {
  id=SoundClass_5
  children [Attenuation_5]
  properties {soundClassOverride=&Fort_WeaponRanged_Fire_Distant}
 }
 {id=WavePlayer_14, properties {soundWaveAssetPtr=&/DMR22_Fire_3P_Close_01}}
 {id=WavePlayer_25, properties {soundWaveAssetPtr=&/DMR22_Fire_3P_Close_02}}
 {id=WavePlayer_26, properties {soundWaveAssetPtr=&/DMR22_Fire_3P_Close_03}}
 {id=WavePlayer_27, properties {soundWaveAssetPtr=&/DMR22_Fire_3P_Close_04}}
 {id=WavePlayer_28, properties {soundWaveAssetPtr=&/DMR22_Fire_3P_Close_05}}
 {id=WavePlayer_29, properties {soundWaveAssetPtr=&/DMR22_Fire_3P_Close_06}}
 {id=WavePlayer_30, properties {soundWaveAssetPtr=&/DMR22_Fire_3P_Medium_01}}
 {id=WavePlayer_31, properties {soundWaveAssetPtr=&/DMR22_Fire_3P_Medium_02}}
 {id=WavePlayer_32, properties {soundWaveAssetPtr=&/DMR22_Fire_3P_Medium_03}}
 {id=WavePlayer_33, properties {soundWaveAssetPtr=&/DMR22_Fire_3P_Medium_04}}
 {id=WavePlayer_34, properties {soundWaveAssetPtr=&/DMR22_Fire_3P_Medium_05}}
 {id=WavePlayer_35, properties {soundWaveAssetPtr=&/DMR22_Fire_3P_Medium_06}}
 {id=WavePlayer_36, properties {soundWaveAssetPtr=&/DMR22_Fire_3P_Distant_01}}
 {id=WavePlayer_37, properties {soundWaveAssetPtr=&/DMR22_Fire_3P_Distant_02}}
 {id=WavePlayer_38, properties {soundWaveAssetPtr=&/DMR22_Fire_3P_Distant_03}}
 {id=WavePlayer_39, properties {soundWaveAssetPtr=&/DMR22_Fire_3P_Distant_04}}
 {id=WavePlayer_40, properties {soundWaveAssetPtr=&/DMR22_Fire_3P_Distant_05}}
 {id=WavePlayer_41, properties {soundWaveAssetPtr=&/DMR22_Fire_3P_Distant_06}}
]
properties {
 bDebug=true
 maxDistance=27500
 soundClassObject=&Fort_WeaponRanged_Fire_3P_PSM
 volumeMultiplier=1
 concurrencyOverrides {bLimitToOwner=true, maxCount=3, resolutionRule=StopOldest}
 concurrencySet [&Weapon_Gunfire_Global_Conc, &Weapon_Gunfire_Local_Conc]
}
