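# Low-ammo warning sound cue for the DMR22 weapon, using first-person weapon attenuation and a degraded low-ammo sound class.
# A Mixer layers the DMR22-specific low-ammo wave with a stock Fortnite rifle low-ammo wave at equal, halved levels.
# A Modulator adds slight upward pitch variation at fixed volume, while the cue lowers overall pitch and boosts volume.
# A concurrency override blocks new instances once the limit is reached.
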SoundCue /CobraDMR/DMR/Sounds/DMR22/DMR22_LowAmmo_Cue
root=Modulator_0
nodes [
 {id=Mixer_0, children [WavePlayer_34, WavePlayer_2], properties {inputVolume [0.5,0.5]}}
 {id=Modulator_0, children [Mixer_0], properties {pitchMin=1, volumeMax=1, volumeMin=1}}
 {id=WavePlayer_2, properties {soundWaveAssetPtr=&/DMR22_LowAmmo}}
 {id=WavePlayer_34, properties {soundWaveAssetPtr=/Game/Sounds/Fort_Weapons/LowAmmo/weapon_low_ammo_hi_rifle_05}}
]
properties {
 attenuationSettings=&Fort_Weapon_Projectile_FirstPerson_Athena_Att
 bOverrideConcurrency=true
 maxDistance=2097152
 pitchMultiplier=0.8
 soundClassObject=/Game/Sounds/FortSoundClassesAndModes/Fort_WeaponRanged_LowAmmo_Degraded_PSM
 volumeMultiplier=1.5
 concurrencyOverrides {resolutionRule=PreventNew}
}
