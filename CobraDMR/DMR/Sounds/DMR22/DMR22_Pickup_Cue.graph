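# Sound cue for the DMR22 weapon pickup sound, classed as a loot-collect sound with weapon-pickup attenuation.
# A quality-level node routes to one of two identically configured mixers, each boosting the same single wave player, so all quality levels sound alike.
# The cue lowers pitch and volume, references an oldest-based pickup concurrency set, and also carries single-instance, prevent-new override values.
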SoundCue /CobraDMR/DMR/Sounds/DMR22/DMR22_Pickup_Cue
root=QualityLevel_0
nodes [
 {id=Mixer_0, children [WavePlayer_0,null], properties {inputVolume [1.2,1]}}
 {id=Mixer_1, children [WavePlayer_0,null], properties {inputVolume [1.2,1]}}
 {id=QualityLevel_0, children [Mixer_1,Mixer_1,Mixer_0,Mixer_0]}
 {id=WavePlayer_0, properties {soundWaveAssetPtr=&/DMR22_Pickup}}
]
properties {
 attenuationSettings=&Fort_Weapon_Pickup_Att
 maxDistance=3500
 pitchMultiplier=0.9
 soundClassObject=&Fort_Loot_Collect
 volumeMultiplier=0.119
 concurrencyOverrides {maxCount=1, resolutionRule=PreventNew}
 concurrencySet [&PickupMaximum_Oldest]
}
