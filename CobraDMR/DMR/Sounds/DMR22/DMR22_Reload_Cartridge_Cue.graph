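# Sound cue for the cartridge foley played when reloading the DMR22 weapon.
# An equally weighted random node picks one of three cartridge waves, which passes through a Fort_WeaponFoley sound class override and a modulator that randomizes volume.
# The root attenuation node applies the weapon reload falloff, also set on the cue itself.
# The cue slightly lowers pitch, overrides concurrency and has its debug flag left on.
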SoundCue /CobraDMR/DMR/Sounds/DMR22/DMR22_Reload_Cartridge_Cue
root=Attenuation_0
nodes [
 {id=Attenuation_0, children [Modulator_0], properties {attenuationSettings=&Fort_Weapon_Reload_Att}}
 {id=Modulator_0, children [SoundClass_0], properties {volumeMax=0.7, volumeMin=0.6}}
 {id=Random_0, children [WavePlayer_1,WavePlayer_2,WavePlayer_3], properties {weights [1,1,1]}}
 {id=SoundClass_0, children [Random_0], properties {soundClassOverride=&Fort_WeaponFoley}}
 {id=WavePlayer_1, properties {soundWaveAssetPtr=&/DMR22_Reload_Cartridge_03}}
 {id=WavePlayer_2, properties {soundWaveAssetPtr=&/DMR22_Reload_Cartridge_01}}
 {id=WavePlayer_3, properties {soundWaveAssetPtr=&/DMR22_Reload_Cartridge_02}}
]
properties {
 attenuationSettings=&Fort_Weapon_Reload_Att
 bDebug=true
 bOverrideConcurrency=true
 maxDistance=2250
 pitchMultiplier=0.9
 soundClassObject=&Fort_WeaponFoley
 volumeMultiplier=1.002
}
